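# Physics asset of cape colliders for the male medium Horned Mask 04 cosmetic body, apparently for cape cloth collision rather than ragdoll, as it has no constraints.
# Capsules cover the pelvis, spine, neck and clavicles, with a sphere on the head and multiple overlapping capsules on spine_02 and spine_05.
# Extra capsules sit on the left bow-tie bone and both outer thigh-cloth bones.
# Arm, thigh, calf and foot bodies hold no shapes.
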PhysicsAsset /BRCosmetics/Characters/Player/Male/Medium/Bodies/M_MED_Horned_Mask_01/Meshes/M_MED_HOrned_Mask_04_CapeColliders
bodies {
 bowTie_01_l [{shapeType=Capsule, length=0.5, radius=17.25, shapeName=bowTie_01_l_capsule, center=<-5.1446 10.5838 -2.9384>, rotation=<rot 0 0 -50.0001>}]
 calf_l []
 calf_r []
 clavicle_l [{shapeType=Capsule, length=7.7636, radius=11.1427, center=<9.9642 -0.7741 -1.5923>, rotation=<rot 75.9352 -0.0001 175.5622>}]
 clavicle_r [{shapeType=Capsule, length=7.7636, radius=11.1427, center=<-9.9642 0.7741 1.5923>, rotation=<rot 75.9351 -0.0001 -4.4378>}]
 foot_l []
 foot_r []
 head [{shapeType=Sphere, radius=17.3125, center=<11.028 -0.174 0>}]
 lowerarm_l []
 lowerarm_r []
 neck_02 [{shapeType=Capsule, length=7.9375, radius=12.8175, center=<-0.2642 -0.3408 0>, rotation=<rot 90 -171.1193 -175.3978>}]
 pelvis [{shapeType=Capsule, length=8.6562, radius=13.5388, center=<-2.6215 -1.3695 0>}]
 spine_01 [{shapeType=Capsule, length=10.7187, radius=12.4138, center=<-2.6215 0.4353 0>}]
 spine_02 [
  {shapeType=Capsule, length=7.5937, radius=13.945, center=<-2.6215 0.4993 0.0001>}
  {shapeType=Capsule, length=29.875, radius=11.1013, center=<2.5024 0.4992 3.9825>, rotation=<rot 90 0.0049 0.0049>}
  {shapeType=Capsule, length=29.875, radius=11.1013, center=<2.4704 0.4991 -3.9825>, rotation=<rot 90 -0.0396 -0.0396>}
 ]
 spine_03 [{shapeType=Capsule, length=6.4062, radius=13.5075, center=<-0.3187 0.7113 0.0001>}]
 spine_04 [{shapeType=Capsule, length=12.9062, radius=15.3825, center=<2.1467 0.0754 0.0002>}]
 spine_05 [
  {shapeType=Capsule, length=19.3437, radius=15.695, center=<-1.7514 0.4083 0>}
  {shapeType=Capsule, length=19.5312, radius=15.5625, center=<0.7875 0.4061 0>}
 ]
 thigh_cloth_out_02_l [{
   shapeType=Capsule
   length=19.25
   radius=15
   shapeName=thigh_cloth_out_02_l_capsule
   center=<1.177 6.6754 -6.7337>
   rotation=<rot 90 -92.5862 -92.5867>
  }]
 thigh_cloth_out_02_r [{
   shapeType=Capsule
   length=20.5
   radius=15
   shapeName=thigh_cloth_out_02_r_capsule
   center=<0.0001 -7.6324 3.7184>
   rotation=<rot -90 179.9978 -179.9969>
  }]
 thigh_l []
 thigh_r []
 upperarm_l []
 upperarm_r []
}
constraints []
