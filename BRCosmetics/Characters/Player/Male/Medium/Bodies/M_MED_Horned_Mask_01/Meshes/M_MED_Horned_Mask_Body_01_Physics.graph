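# Physics asset for the medium male Horned Mask 01 player cosmetic body, covering its left and right dynamic belt bones plus the spine and thighs.
# The right belt joins spine_01 with angle limits, while the left belt joins a small master body that connects to spine_01 with free swing and twist.
# All bodies are single capsules; the thighs have no constraints, apparently serving as colliders for the belts.
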PhysicsAsset /BRCosmetics/Characters/Player/Male/Medium/Bodies/M_MED_Horned_Mask_01/Meshes/M_MED_Horned_Mask_Body_01_Physics
bodies {
 dyn_belt_01_l [{shapeType=Capsule, length=2.7992, radius=5.7466, center=<12.0463 -0.0164 -0.563>, rotation=<rot 87.2959 -0.0008 -0.0008>}]
 dyn_belt_01_r [{shapeType=Capsule, length=2.9444, radius=5.6015, center=<12.0462 -0.6388 -1.9997>, rotation=<rot 90 15.1541 19.7174>}]
 dyn_belt_l_MSTR [{shapeType=Capsule, length=0.4068, radius=0.8638, center=<-0.1333 -0.2532 0.4341>, rotation=<rot 0 -87.0845 -86.2344>}]
 spine_01 [{shapeType=Capsule, length=9.6831, radius=13.9475}]
 thigh_l [{shapeType=Capsule, length=17.6035, radius=10.3785, center=<-22.0286 2.9722 -0.8756>, rotation=<rot 90 0 0>}]
 thigh_r [{shapeType=Capsule, length=18.1035, radius=9.6285, center=<20.2554 -3.6802 -0.307>, rotation=<rot 84.903 179.7209 179.7219>}]
}
constraints [
 {bone1Name=dyn_belt_01_l, bone2Name=dyn_belt_l_MSTR, swing1LimitDegrees=45, swing2LimitDegrees=45, twistLimitDegrees=15}
 {bone1Name=dyn_belt_01_r, bone2Name=spine_01, swing1LimitDegrees=45, swing2LimitDegrees=45, twistLimitDegrees=15}
 {
  bone1Name=dyn_belt_l_MSTR
  bone2Name=spine_01
  swing1LimitDegrees=45
  swing1Motion=Free
  swing2LimitDegrees=45
  swing2Motion=Free
  twistLimitDegrees=45
  twistMotion=Free
 }
]
