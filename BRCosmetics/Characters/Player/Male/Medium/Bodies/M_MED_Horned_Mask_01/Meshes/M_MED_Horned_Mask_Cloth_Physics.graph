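# Physics asset for the medium male Horned Mask 01 player cosmetic, setting up bone-driven cloth panels over both thighs.
# Capsules on pelvis, thighs and calves form the body volumes, with three capsules per thigh.
# Each side has center, inner and outer three-box chains hanging from the thigh with limited swing and twist.
# Looser cross-constraints tie neighboring chains row by row into a sheet; the left side carries some duplicate reversed links.
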PhysicsAsset /BRCosmetics/Characters/Player/Male/Medium/Bodies/M_MED_Horned_Mask_01/Meshes/M_MED_Horned_Mask_Cloth_Physics
bodies {
 calf_l [{shapeType=Capsule, length=25.2659, radius=7.4457, center=<-8.3067 -0.7451 -0.8191>, rotation=<rot 90 0 0>}]
 calf_r [{shapeType=Capsule, length=20.2659, radius=7.9457, center=<8.8234 0.745 0.8191>, rotation=<rot 90 0 0>}]
 pelvis [{shapeType=Capsule, length=13.75, radius=9.4312, center=<-5.5306 0.6557 -0.727>}]
 thigh_cloth_01_l [{shapeType=Box, extentX=14, extentY=10, extentZ=2, center=<7.6029 0.4128 0>}]
 thigh_cloth_01_r [{shapeType=Box, extentX=14, extentY=10, extentZ=2, center=<-7.6029 -0.4128 0>, rotation=<rot 0 0 -180>}]
 thigh_cloth_02_l [{shapeType=Box, extentX=14, extentY=10, extentZ=2, center=<7.6029 0.4128 0>}]
 thigh_cloth_02_r [{shapeType=Box, extentX=14, extentY=10, extentZ=2, center=<-7.6029 -0.4128 0>, rotation=<rot 0 0 -180>}]
 thigh_cloth_03_l [{shapeType=Box, extentX=8.5, extentY=9.5, extentZ=2, center=<5 0 0>}]
 thigh_cloth_03_r [{shapeType=Box, extentX=8.5, extentY=9.5, extentZ=2, center=<-5 0 0>, rotation=<rot 0 0 -180>}]
 thigh_cloth_in_01_l [{shapeType=Box, extentX=14, extentY=10, extentZ=3.6225, center=<7.6003 0.5349 0.4377>, rotation=<rot -4.9809 -0.4365 5.0189>}]
 thigh_cloth_in_01_r [{shapeType=Box, extentX=14, extentY=10, extentZ=3.6225, center=<-7.6003 -0.5349 -0.4377>, rotation=<rot -4.9809 -0.4365 -174.9811>}]
 thigh_cloth_in_02_l [{shapeType=Box, extentX=14, extentY=10, extentZ=3.5057, center=<7.5782 -1.4215 0.1865>}]
 thigh_cloth_in_02_r [{shapeType=Box, extentX=14, extentY=10, extentZ=3.5057, center=<-7.5782 1.4215 -0.1865>, rotation=<rot 0 0 -180>}]
 thigh_cloth_in_03_l [{shapeType=Box, extentX=8.5, extentY=9, extentZ=3.2469, center=<5 -1.0454 0>}]
 thigh_cloth_in_03_r [{shapeType=Box, extentX=8.5, extentY=9, extentZ=3.2469, center=<-5 1.0454 0>, rotation=<rot 0 0 -180>}]
 thigh_cloth_out_01_l [{shapeType=Box, extentX=14, extentY=10, extentZ=2.7272, center=<7.6354 0 0>, rotation=<rot -4.9238 0.8705 -10.0375>}]
 thigh_cloth_out_01_r [{shapeType=Box, extentX=14, extentY=10, extentZ=2.7272, center=<-7.6354 0 0>, rotation=<rot -4.9238 0.8705 169.9625>}]
 thigh_cloth_out_02_l [{shapeType=Box, extentX=14, extentY=10, extentZ=3.3822, center=<5.6779 -0.0336 0.3843>, rotation=<rot 0 0.0002 -15>}]
 thigh_cloth_out_02_r [{shapeType=Box, extentX=14, extentY=10, extentZ=3.3822, center=<-5.6779 0.0336 -0.3843>, rotation=<rot 0 0.0002 165>}]
 thigh_cloth_out_03_l [{shapeType=Box, extentX=9, extentY=8.5, extentZ=2, center=<5 0 0.3119>}]
 thigh_cloth_out_03_r [{shapeType=Box, extentX=9, extentY=8.5, extentZ=2, center=<-5 0 -0.3119>, rotation=<rot 0 0 -180>}]
 thigh_l [
  {shapeType=Capsule, length=20.2839, radius=7.4805, center=<-26.8505 -0.6271 -4.7806>, rotation=<rot 89.7314 0.0078 170.0081>}
  {shapeType=Capsule, length=11.6168, radius=8.9685, center=<-23.0713 -3.9424 -2.0569>, rotation=<rot 85.2643 -169.4387 -4.436>}
  {shapeType=Capsule, length=8.4562, radius=9.6818, center=<-23.6087 5.4157 -1.2825>, rotation=<rot 85.2298 -174.6818 -9.642>}
 ]
 thigh_r [
  {shapeType=Capsule, length=20.2839, radius=7.4805, center=<26.8505 0.6271 4.7806>, rotation=<rot 89.7314 0.0078 -9.9919>}
  {shapeType=Capsule, length=11.6168, radius=8.9685, center=<23.0713 3.9424 2.0569>, rotation=<rot 85.2643 -169.4387 175.564>}
  {shapeType=Capsule, length=8.4562, radius=9.6818, center=<23.6087 -5.4157 1.2825>, rotation=<rot 85.2298 -174.6818 170.358>}
 ]
}
constraints [
 {bone1Name=calf_l, bone2Name=thigh_l, swing1LimitDegrees=45, swing2LimitDegrees=45, twistLimitDegrees=45}
 {bone1Name=calf_r, bone2Name=thigh_r, swing1LimitDegrees=45, swing2LimitDegrees=45, twistLimitDegrees=45}
 {bone1Name=thigh_l, bone2Name=pelvis, swing1LimitDegrees=45, swing2LimitDegrees=45, twistLimitDegrees=15}
 {bone1Name=thigh_r, bone2Name=pelvis, swing1LimitDegrees=45, swing2LimitDegrees=45, twistLimitDegrees=15}
 {bone1Name=thigh_cloth_01_r, bone2Name=thigh_r, swing1LimitDegrees=30, swing2LimitDegrees=20, twistLimitDegrees=10}
 {bone1Name=thigh_cloth_01_l, bone2Name=thigh_l, swing1LimitDegrees=30, swing2LimitDegrees=20, twistLimitDegrees=10}
 {bone1Name=thigh_cloth_02_l, bone2Name=thigh_cloth_01_l, swing1LimitDegrees=30, swing2LimitDegrees=30, twistLimitDegrees=10}
 {bone1Name=thigh_cloth_02_r, bone2Name=thigh_cloth_01_r, swing1LimitDegrees=30, swing2LimitDegrees=30, twistLimitDegrees=10}
 {bone1Name=thigh_cloth_in_01_l, bone2Name=thigh_l, swing1LimitDegrees=30, swing2LimitDegrees=20, twistLimitDegrees=10}
 {bone1Name=thigh_cloth_in_02_l, bone2Name=thigh_cloth_in_01_l, swing1LimitDegrees=30, swing2LimitDegrees=30, twistLimitDegrees=10}
 {bone1Name=thigh_cloth_in_03_l, bone2Name=thigh_cloth_in_02_l, swing1LimitDegrees=30, swing2LimitDegrees=30, twistLimitDegrees=10}
 {bone1Name=thigh_cloth_out_01_l, bone2Name=thigh_l, swing1LimitDegrees=30, swing2LimitDegrees=20, twistLimitDegrees=10}
 {bone1Name=thigh_cloth_out_02_l, bone2Name=thigh_cloth_out_01_l, swing1LimitDegrees=30, swing2LimitDegrees=30, twistLimitDegrees=10}
 {bone1Name=thigh_cloth_out_03_l, bone2Name=thigh_cloth_out_02_l, swing1LimitDegrees=30, swing2LimitDegrees=30, twistLimitDegrees=10}
 {bone1Name=thigh_cloth_in_01_r, bone2Name=thigh_r, swing1LimitDegrees=30, swing2LimitDegrees=20, twistLimitDegrees=10}
 {bone1Name=thigh_cloth_in_02_r, bone2Name=thigh_cloth_in_01_r, swing1LimitDegrees=30, swing2LimitDegrees=30, twistLimitDegrees=10}
 {bone1Name=thigh_cloth_in_03_r, bone2Name=thigh_cloth_in_02_r, swing1LimitDegrees=30, swing2LimitDegrees=30, twistLimitDegrees=10}
 {bone1Name=thigh_cloth_out_01_r, bone2Name=thigh_r, swing1LimitDegrees=30, swing2LimitDegrees=20, twistLimitDegrees=10}
 {bone1Name=thigh_cloth_out_02_r, bone2Name=thigh_cloth_out_01_r, swing1LimitDegrees=30, swing2LimitDegrees=30, twistLimitDegrees=10}
 {bone1Name=thigh_cloth_out_03_r, bone2Name=thigh_cloth_out_02_r, swing1LimitDegrees=30, swing2LimitDegrees=30, twistLimitDegrees=10}
 {bone1Name=thigh_cloth_03_l, bone2Name=thigh_cloth_02_l, swing1LimitDegrees=30, swing2LimitDegrees=30, twistLimitDegrees=10}
 {bone1Name=thigh_cloth_03_r, bone2Name=thigh_cloth_02_r, swing1LimitDegrees=30, swing2LimitDegrees=30, twistLimitDegrees=10}
 {bone1Name=thigh_cloth_03_l, bone2Name=thigh_cloth_out_03_l, swing1LimitDegrees=40, swing2LimitDegrees=40, twistLimitDegrees=10}
 {bone1Name=thigh_cloth_01_l, bone2Name=thigh_cloth_in_01_l, swing1LimitDegrees=40, swing2LimitDegrees=40, twistLimitDegrees=10}
 {bone1Name=thigh_cloth_01_l, bone2Name=thigh_cloth_out_01_l, swing1LimitDegrees=40, swing2LimitDegrees=40, twistLimitDegrees=10}
 {bone1Name=thigh_cloth_in_03_l, bone2Name=thigh_cloth_03_l, swing1LimitDegrees=40, swing2LimitDegrees=40, twistLimitDegrees=10}
 {bone1Name=thigh_cloth_out_03_l, bone2Name=thigh_cloth_03_l, swing1LimitDegrees=40, swing2LimitDegrees=40, twistLimitDegrees=10}
 {bone1Name=thigh_cloth_in_02_l, bone2Name=thigh_cloth_02_l, swing1LimitDegrees=40, swing2LimitDegrees=40, twistLimitDegrees=10}
 {bone1Name=thigh_cloth_out_02_l, bone2Name=thigh_cloth_02_l, swing1LimitDegrees=40, swing2LimitDegrees=40, twistLimitDegrees=10}
 {bone1Name=thigh_cloth_in_01_l, bone2Name=thigh_cloth_01_l, swing1LimitDegrees=40, swing2LimitDegrees=40, twistLimitDegrees=10}
 {bone1Name=thigh_cloth_out_01_l, bone2Name=thigh_cloth_01_l, swing1LimitDegrees=40, swing2LimitDegrees=40, twistLimitDegrees=10}
 {bone1Name=thigh_cloth_in_03_r, bone2Name=thigh_cloth_03_r, swing1LimitDegrees=40, swing2LimitDegrees=40, twistLimitDegrees=10}
 {bone1Name=thigh_cloth_out_03_r, bone2Name=thigh_cloth_03_r, swing1LimitDegrees=40, swing2LimitDegrees=40, twistLimitDegrees=10}
 {bone1Name=thigh_cloth_in_02_r, bone2Name=thigh_cloth_02_r, swing1LimitDegrees=40, swing2LimitDegrees=40, twistLimitDegrees=10}
 {bone1Name=thigh_cloth_out_02_r, bone2Name=thigh_cloth_02_r, swing1LimitDegrees=40, swing2LimitDegrees=40, twistLimitDegrees=10}
 {bone1Name=thigh_cloth_in_01_r, bone2Name=thigh_cloth_01_r, swing1LimitDegrees=40, swing2LimitDegrees=40, twistLimitDegrees=10}
 {bone1Name=thigh_cloth_out_01_r, bone2Name=thigh_cloth_01_r, swing1LimitDegrees=40, swing2LimitDegrees=40, twistLimitDegrees=10}
]
