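# Post-process-style AnimBlueprint that adds secondary motion to the Horned Mask male body cosmetic.
# It copies the attached parent's pose, then a neck-driven Pose Driver corrects the collar bones.
# AnimDynamics chains swing bow, bow-tie and shoulder fur bones, with partial rotation constraints making fur follow upper-arm twist.
# Cloth and body RigidBody passes follow, separated by a crouch-only left belt offset; all simulations receive skydiving force.
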AnimBlueprint /BRCosmetics/Characters/Player/Male/Medium/Bodies/M_MED_Horned_Mask_01/Meshes/M_MED_Horned_Mask_Body_01_AnimBP
target_skeleton=&/M_MED_Horned_Mask_Body_01_Skeleton
graphs [
 AnimGraph {nodes [
   n1 = Break Fort Anim Input Skydiving External Force(FortAnimInput_SkydivingExternalForce=Get SkydivingExternalForceInput()) -> (
    bUseSkydivingVectorForce=~
    SkydivingVectorForce=~
    HeadToPelvisDirection=~
    FloatingMultiplier=~
    FloatingAdditive=~
    DivingMultiplier=~
    DivingAdditive=~
    ParachutingMultiplier=~
    ParachutingAdditive=~
   )
   n2 = Pose Driver - Source: neck_02!(SourcePose:Pose Link Structure=Copy Pose From Mesh() {bUseAttachedParent=true}) -> (Pose:Pose Link Structure) {
    is=title AnimGraphNode_PoseDriver
    poseAsset=&/HM_Neck_Pose_PoseAsset
    onlyDriveBones [{boneName=collar_back}, {boneName=collar_l}, {boneName=collar_r}]
    poseTargets [
     {
      bApplyCustomCurve=false
      bIsHidden=false
      distanceMethod=DefaultMethod
      drivenName=Base
      functionType=DefaultFunction
      targetScale=1
      boneTransforms [
       {targetRotation=<rot 0 -5.7813 0>, targetTranslation=<4.5537 0.0001 -0.0002>}
       {targetRotation=<rot 0 23.1062 0>, targetTranslation=<11.2914 0.1474 -0.0002>}
       {targetRotation=<rot 0 -13.2022 0>, targetTranslation=<5.3812 -0.0012 0>}
      ]
      customCurve {defaultValue=3.4028234663852886e+38, postInfinityExtrap=RCCE_Constant, preInfinityExtrap=RCCE_Constant, keys []}
      targetRotation=<rot 0 0 0>
     }
     {
      bApplyCustomCurve=false
      bIsHidden=false
      distanceMethod=DefaultMethod
      drivenName=Back
      functionType=DefaultFunction
      targetScale=0.8536
      boneTransforms [
       {targetRotation=<rot -0.0223 -5.6045 0.5328>, targetTranslation=<4.5537 0.0001 -0.0002>}
       {targetRotation=<rot -1.6835 0.1941 15.199>, targetTranslation=<11.2914 0.1474 -0.0002>}
       {targetRotation=<rot 3.2374 -36.0975 2.9157>, targetTranslation=<5.3812 -0.0012 0>}
      ]
      customCurve {defaultValue=3.4028234663852886e+38, postInfinityExtrap=RCCE_Constant, preInfinityExtrap=RCCE_Constant, keys []}
      targetRotation=<rot 0 0 0>
     }
     {
      bApplyCustomCurve=false
      bIsHidden=false
      distanceMethod=DefaultMethod
      drivenName=Back_Left
      functionType=DefaultFunction
      targetScale=0.8536
      boneTransforms [
       {targetRotation=<rot 1.5979 -4.0372 -5.7399>, targetTranslation=<4.567 -0.0026 -0.0004>}
       {targetRotation=<rot 14.9205 1.8004 -18.9091>, targetTranslation=<11.3135 0.1525 0.0006>}
       {targetRotation=<rot 3.5515 -20.2218 -40.0036>, targetTranslation=<5.3964 0.0003 -0.0003>}
      ]
      customCurve {defaultValue=3.4028234663852886e+38, postInfinityExtrap=RCCE_Constant, preInfinityExtrap=RCCE_Constant, keys []}
      targetRotation=<rot 0 0 0>
     }
     {
      bApplyCustomCurve=false
      bIsHidden=false
      distanceMethod=DefaultMethod
      drivenName=Left
      functionType=DefaultFunction
      targetScale=0.891
      boneTransforms [
       {targetRotation=<rot -1.352 -5.5122 3.9225>, targetTranslation=<4.5585 0.0001 0.0021>}
       {targetRotation=<rot -14.8129 -3.8884 19.6621>, targetTranslation=<11.254 0.1495 -0.0031>}
       {targetRotation=<rot 3.2934 -16.2667 44.5532>, targetTranslation=<5.3862 0.0007 0.0026>}
      ]
      customCurve {defaultValue=3.4028234663852886e+38, postInfinityExtrap=RCCE_Constant, preInfinityExtrap=RCCE_Constant, keys []}
      targetRotation=<rot 0 0 0>
     }
     {
      bApplyCustomCurve=false
      bIsHidden=false
      distanceMethod=DefaultMethod
      drivenName=Right
      functionType=DefaultFunction
      targetScale=0.9804
      boneTransforms [
       {targetRotation=<rot -0.0052 -5.8044 0.1074>, targetTranslation=<4.5591 0.0005 0.0014>}
       {targetRotation=<rot 18.9769 13.2592 14.741>, targetTranslation=<11.3034 0.1532 -0.0001>}
       {targetRotation=<rot 15.4639 -41.071 3.4476>, targetTranslation=<5.387 -0.0004 0>}
      ]
      customCurve {defaultValue=3.4028234663852886e+38, postInfinityExtrap=RCCE_Constant, preInfinityExtrap=RCCE_Constant, keys []}
      targetRotation=<rot 0 0 0>
     }
    ]
    rBFParams {radius=20}
    sourceBones [{boneName=neck_02}, {boneName=neck_01}, {boneName=head}]
   }
   n3 = Anim Dynamics - Bone: bowTie_01_l(ComponentPose=Local To Component(LocalPose=@n2), Alpha=1, ExternalForce=@n1.SkydivingVectorForce) {
    angularBiasOverride=1
    angularDampingOverride=0.98
    bGravityOverrideInSimSpace=true
    gravityScale=0.32
    lODThreshold=1
    numSolverIterationsPostUpdate=4
    numSolverIterationsPreUpdate=16
    simulationSpace=RootRelative
    alphaBoolBlend {blendOption=Linear}
    boundBone {boneName=bowTie_01_l}
    componentLinearAccScale=<0.5 0.5 0.5>
    physicsBodyDefinitions [{
      sphereCollisionRadius=0
      boundBone {boneName=bowTie_01_l}
      boxExtents=<10 10 5>
      constraintSetup {angularLimitsMax=<15 20 5>, angularLimitsMin=<0 -20 -5>, angularTarget=<0 0 0>}
      localJointOffset=<0.5 0 0>
     }]
   }
   n4 = Anim Dynamics - Bone: bowTie_02_l(ComponentPose=@n3, Alpha=1, ExternalForce=@n1.SkydivingVectorForce) {
    angularBiasOverride=1
    angularDampingOverride=0.98
    bGravityOverrideInSimSpace=true
    gravityScale=0.32
    lODThreshold=1
    numSolverIterationsPostUpdate=4
    numSolverIterationsPreUpdate=16
    simulationSpace=RootRelative
    alphaBoolBlend {blendOption=Linear}
    boundBone {boneName=bowTie_02_l}
    componentLinearAccScale=<0.5 0.5 0.5>
    physicsBodyDefinitions [{
      sphereCollisionRadius=0
      boundBone {boneName=bowTie_02_l}
      boxExtents=<10 10 5>
      constraintSetup {angularLimitsMax=<10 20 5>, angularLimitsMin=<0 -20 -5>, angularTarget=<0 0 0>}
      localJointOffset=<0.5 0 0>
     }]
   }
   n5 = Anim Dynamics - Bone: bow_01_l(ComponentPose=@n4, Alpha=1, ExternalForce=@n1.SkydivingVectorForce) {
    angularBiasOverride=1
    angularDampingOverride=0.99
    bGravityOverrideInSimSpace=true
    bOverrideAngularDamping=true
    gravityScale=0.32
    lODThreshold=1
    numSolverIterationsPostUpdate=4
    numSolverIterationsPreUpdate=16
    alphaBoolBlend {blendOption=Linear}
    boundBone {boneName=bow_01_l}
    componentLinearAccScale=<1 1 1>
    componentLinearVelScale=<1 1 1>
    physicsBodyDefinitions [{
      sphereCollisionRadius=0
      boundBone {boneName=bow_01_l}
      boxExtents=<12 5 12>
      constraintSetup {angularLimitsMax=<5 20 5>, angularLimitsMin=<-5 -20 -5>, angularTarget=<0 0 0>}
      localJointOffset=<1 0 0>
     }]
   }
   n6 = Anim Dynamics - Bone: bow_02_l(ComponentPose=@n5, Alpha=1, ExternalForce=@n1.SkydivingVectorForce) {
    angularBiasOverride=1
    angularDampingOverride=0.99
    bGravityOverrideInSimSpace=true
    bOverrideAngularDamping=true
    gravityScale=0.32
    lODThreshold=1
    numSolverIterationsPostUpdate=4
    numSolverIterationsPreUpdate=16
    alphaBoolBlend {blendOption=Linear}
    boundBone {boneName=bow_02_l}
    componentLinearAccScale=<1 1 1>
    componentLinearVelScale=<1 1 1>
    physicsBodyDefinitions [{
      sphereCollisionRadius=0
      boundBone {boneName=bow_02_l}
      boxExtents=<12 5 12>
      constraintSetup {angularLimitsMax=<5 35 5>, angularLimitsMin=<-5 -35 -5>, angularTarget=<0 0 0>}
      localJointOffset=<1 0 0>
     }]
   }
   n7 = Anim Dynamics - Bone: bowTie_01_r(ComponentPose=@n6, Alpha=1, ExternalForce=@n1.SkydivingVectorForce) {
    angularBiasOverride=1
    angularDampingOverride=0.98
    bGravityOverrideInSimSpace=true
    gravityScale=0.32
    lODThreshold=1
    numSolverIterationsPostUpdate=4
    numSolverIterationsPreUpdate=16
    simulationSpace=RootRelative
    alphaBoolBlend {blendOption=Linear}
    boundBone {boneName=bowTie_01_r}
    componentLinearAccScale=<0.5 0.5 0.5>
    physicsBodyDefinitions [{
      sphereCollisionRadius=0
      boundBone {boneName=bowTie_01_r}
      boxExtents=<10 10 5>
      constraintSetup {angularLimitsMax=<15 20 5>, angularLimitsMin=<0 -20 -5>, angularTarget=<0 0 0>}
      localJointOffset=<0.5 0 0>
     }]
   }
   n8 = Anim Dynamics - Bone: bowTie_02_r(ComponentPose=@n7, Alpha=1, ExternalForce=@n1.SkydivingVectorForce) {
    angularBiasOverride=1
    angularDampingOverride=0.98
    bGravityOverrideInSimSpace=true
    gravityScale=0.32
    lODThreshold=1
    numSolverIterationsPostUpdate=4
    numSolverIterationsPreUpdate=16
    simulationSpace=RootRelative
    alphaBoolBlend {blendOption=Linear}
    boundBone {boneName=bowTie_02_r}
    componentLinearAccScale=<0.5 0.5 0.5>
    physicsBodyDefinitions [{
      sphereCollisionRadius=0
      boundBone {boneName=bowTie_02_r}
      boxExtents=<10 10 5>
      constraintSetup {angularLimitsMax=<10 20 5>, angularLimitsMin=<0 -20 -5>, angularTarget=<0 0 0>}
      localJointOffset=<0.5 0 0>
     }]
   }
   n9 = Anim Dynamics - Bone: bow_01_r(ComponentPose=@n8, Alpha=1, ExternalForce=@n1.SkydivingVectorForce) {
    angularBiasOverride=1
    angularDampingOverride=0.99
    bGravityOverrideInSimSpace=true
    bOverrideAngularDamping=true
    gravityScale=0.32
    lODThreshold=1
    numSolverIterationsPostUpdate=4
    numSolverIterationsPreUpdate=16
    alphaBoolBlend {blendOption=Linear}
    boundBone {boneName=bow_01_r}
    componentLinearAccScale=<1 1 1>
    componentLinearVelScale=<1 1 1>
    physicsBodyDefinitions [{
      sphereCollisionRadius=0
      boundBone {boneName=bow_01_r}
      boxExtents=<12 5 12>
      constraintSetup {angularLimitsMax=<5 35 5>, angularLimitsMin=<-5 0 -5>, angularTarget=<0 0 0>}
      localJointOffset=<1 0 0>
     }]
   }
   n10 = Anim Dynamics - Bone: bow_02_r(ComponentPose=@n9, Alpha=1, ExternalForce=@n1.SkydivingVectorForce) {
    angularBiasOverride=1
    angularDampingOverride=0.99
    bGravityOverrideInSimSpace=true
    bOverrideAngularDamping=true
    gravityScale=0.32
    lODThreshold=1
    numSolverIterationsPostUpdate=4
    numSolverIterationsPreUpdate=16
    alphaBoolBlend {blendOption=Linear}
    boundBone {boneName=bow_02_r}
    componentLinearAccScale=<1 1 1>
    componentLinearVelScale=<1 1 1>
    physicsBodyDefinitions [{
      sphereCollisionRadius=0
      boundBone {boneName=bow_02_r}
      boxExtents=<12 5 12>
      constraintSetup {angularLimitsMax=<5 35 5>, angularLimitsMin=<-5 -15 -5>, angularTarget=<0 0 0>}
      localJointOffset=<1 0 0>
     }]
   }
   n11 = "Constraint - shoulder_fur_r (R)"(ComponentPose=@n10, Alpha=0.25, ConstraintWeights_0=1) {
    alphaBoolBlend {blendOption=Linear}
    boneToModify {boneName=shoulder_fur_r}
    constraintSetup [{transformType=Rotation, targetBone {boneName=upperarm_twist_01_r}}]
    constraintWeights [1]
   }
   n12 = Anim Dynamics - Bone: dyn_shoulder_fur_l(ComponentPose=@n11, Alpha=1, ExternalForce=@n1.SkydivingVectorForce) {
    angularDampingOverride=0.98
    angularSpringConstant=500
    bAngularSpring=true
    bGravityOverrideInSimSpace=true
    bOverrideAngularDamping=true
    gravityScale=0.32
    simulationSpace=RootRelative
    alphaBoolBlend {blendOption=Linear}
    boundBone {boneName=dyn_shoulder_fur_l}
    physicsBodyDefinitions [{
      sphereCollisionRadius=0
      boundBone {boneName=dyn_shoulder_fur_l}
      boxExtents=<10 10 3>
      constraintSetup {angularLimitsMax=<0 5 10>, angularLimitsMin=<0 -5 -10>, angularTarget=<0 0 0>}
      localJointOffset=<-0.1 0 0>
     }]
   }
   n13 = "Constraint - shoulder_fur_l (R)"(ComponentPose=@n12, Alpha=0.25, ConstraintWeights_0=1) {
    alphaBoolBlend {blendOption=Linear}
    boneToModify {boneName=shoulder_fur_l}
    constraintSetup [{transformType=Rotation, targetBone {boneName=upperarm_twist_01_l}}]
    constraintWeights [1]
   }
   n14 = Anim Dynamics - Bone: dyn_shoulder_fur_r(ComponentPose=@n13, Alpha=1, ExternalForce=@n1.SkydivingVectorForce) {
    angularDampingOverride=0.98
    angularSpringConstant=500
    bAngularSpring=true
    bGravityOverrideInSimSpace=true
    bOverrideAngularDamping=true
    gravityScale=0.32
    simulationSpace=RootRelative
    alphaBoolBlend {blendOption=Linear}
    boundBone {boneName=dyn_shoulder_fur_r}
    physicsBodyDefinitions [{
      sphereCollisionRadius=0
      boundBone {boneName=dyn_shoulder_fur_r}
      boxExtents=<10 10 3>
      constraintSetup {angularLimitsMax=<0 5 10>, angularLimitsMin=<0 -5 -10>, angularTarget=<0 0 0>}
      localJointOffset=<0.1 0 0>
     }]
   }
   n15 = RigidBody!(ComponentPose=@n14, bAlphaBoolEnabled:bool, ExternalForce=@n1.SkydivingVectorForce) -> (Pose) {
    alphaInputType=Bool
    bClampLinearTranslationLimitToRefPose=true
    bOverrideWorldGravity=true
    overridePhysicsAsset=&/M_MED_Horned_Mask_Cloth_Physics
    simulationSpace=BaseBoneSpace
    alphaBoolBlend {blendOption=Linear}
    baseBoneRef {boneName=root}
    componentLinearAccScale=<0.2 0.2 0.2>
    componentLinearVelScale=<0.2 0.2 0.2>
    overrideWorldGravity {z=-980}
   }
   n16 = "Transform (Modify) Bone - Bone: dyn_belt_l_MSTR"!(ComponentPose=@n15, bAlphaBoolEnabled:bool=Get bIsCrouching(), Translation="-3.000,0.000,7", Rotation="0,0,0", Scale="1,1,1") -> (Pose) {
    alphaInputType=Bool
    rotationMode=BMM_Additive
    translationMode=BMM_Additive
    alphaBoolBlend {blendOption=Linear}
    boneToModify {boneName=dyn_belt_l_MSTR}
    translation {x=-3, z=7}
   }
   n17 = RigidBody(ComponentPose=@n16, ExternalForce=@n1.SkydivingVectorForce) {
    alphaInputType=Bool
    bClampLinearTranslationLimitToRefPose=true
    bOverrideWorldGravity=true
    overridePhysicsAsset=&/M_MED_Horned_Mask_Body_01_Physics
    simulationSpace=BaseBoneSpace
    alphaBoolBlend {blendOption=Linear}
    baseBoneRef {boneName=root}
    componentLinearAccScale {x=0.2, z=0.2}
    componentLinearVelScale=<0.2 0.2 0.2>
    overrideWorldGravity {z=-980}
   }
   Output Pose(Result=Component To Local(ComponentPose=@n17)) {layerGroup=None}
  ]}
 EventGraph {links [], nodes []}
]
variables []
