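# Physics asset for the face accessory mesh of the male medium Horned Mask 01 cosmetic body, defining physics bodies for its dynamic hair bones.
# Eleven individually sized capsules cover the dyn_hair bones, alongside one large capsule on the head.
# Each hair body is constrained directly to the head rather than chained to another, with identical angular limits that allow moderate swing and tighter twist.
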PhysicsAsset /BRCosmetics/Characters/Player/Male/Medium/Bodies/M_MED_Horned_Mask_01/Meshes/M_MED_Horned_Mask_FaceAcc_01_Physics
bodies {
 dyn_hair_01 [{shapeType=Capsule, length=3.0662, radius=3.0331, center=<0.6352 -0.0001 0>, rotation=<rot 90 0 0>}]
 dyn_hair_010 [{shapeType=Capsule, length=1.4793, radius=1.0707, center=<0.6352 -0.0001 0>, rotation=<rot 90 0 0>}]
 dyn_hair_011 [{shapeType=Capsule, length=0.8621, radius=0.8236, center=<0.6352 -0.0001 0>, rotation=<rot 90 0 0>}]
 dyn_hair_02 [{shapeType=Capsule, length=7.3589, radius=4.0304, center=<0.6352 -0.0001 0>, rotation=<rot 90 0 0>}]
 dyn_hair_03 [{shapeType=Capsule, length=3.9797, radius=3.0281, center=<0.6352 -0.0001 0>, rotation=<rot 90 0 0>}]
 dyn_hair_04 [{shapeType=Capsule, length=5.3425, radius=2.8996, center=<0.6352 -0.0001 0>, rotation=<rot 90 0 0>}]
 dyn_hair_05 [{shapeType=Capsule, length=8.3188, radius=2.6779, center=<0.6352 -0.0001 0>, rotation=<rot 90 0 0>}]
 dyn_hair_06 [{shapeType=Capsule, length=4.3249, radius=3.5678, center=<0.6352 -0.0001 0>, rotation=<rot 90 0 0>}]
 dyn_hair_07 [{shapeType=Capsule, length=2.7286, radius=2.198, center=<0.6352 -0.0001 0>, rotation=<rot 90 0 0>}]
 dyn_hair_08 [{shapeType=Capsule, length=2.9523, radius=2.7479, center=<0.6352 -0.0001 0>, rotation=<rot 90 0 0>}]
 dyn_hair_09 [{shapeType=Capsule, length=1.0213, radius=0.9848, center=<0.6352 -0.0001 0>, rotation=<rot 90 0 0>}]
 head [{shapeType=Capsule, length=10.5526, radius=10.1329, center=<3.1183 6.0111 0.0037>, rotation=<rot 0 100.715 -89.9999>}]
}
constraints [
 {bone1Name=dyn_hair_01, bone2Name=head, swing1LimitDegrees=40, swing2LimitDegrees=40, twistLimitDegrees=10}
 {bone1Name=dyn_hair_02, bone2Name=head, swing1LimitDegrees=40, swing2LimitDegrees=40, twistLimitDegrees=10}
 {bone1Name=dyn_hair_03, bone2Name=head, swing1LimitDegrees=40, swing2LimitDegrees=40, twistLimitDegrees=10}
 {bone1Name=dyn_hair_04, bone2Name=head, swing1LimitDegrees=40, swing2LimitDegrees=40, twistLimitDegrees=10}
 {bone1Name=dyn_hair_05, bone2Name=head, swing1LimitDegrees=40, swing2LimitDegrees=40, twistLimitDegrees=10}
 {bone1Name=dyn_hair_06, bone2Name=head, swing1LimitDegrees=40, swing2LimitDegrees=40, twistLimitDegrees=10}
 {bone1Name=dyn_hair_07, bone2Name=head, swing1LimitDegrees=40, swing2LimitDegrees=40, twistLimitDegrees=10}
 {bone1Name=dyn_hair_08, bone2Name=head, swing1LimitDegrees=40, swing2LimitDegrees=40, twistLimitDegrees=10}
 {bone1Name=dyn_hair_09, bone2Name=head, swing1LimitDegrees=40, swing2LimitDegrees=40, twistLimitDegrees=10}
 {bone1Name=dyn_hair_010, bone2Name=head, swing1LimitDegrees=40, swing2LimitDegrees=40, twistLimitDegrees=10}
 {bone1Name=dyn_hair_011, bone2Name=head, swing1LimitDegrees=40, swing2LimitDegrees=40, twistLimitDegrees=10}
]
